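# Post-style Anim Blueprint for the PrawnTwine male medium body cosmetic, adding secondary tail motion on top of the pose copied from the attached parent mesh.
# The copied pose runs through a Control Rig fed the emote state, then chained spring-damped AnimDynamics simulate four tail bones.
# An additive Transform Bone rescales base_tail when a cape or long backbling is worn, detected when character customization loads.
# RBANOff and isWearingPack are set but unused in the AnimGraph.
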AnimBlueprint /BRCosmetics/Characters/Player/Male/Medium/Bodies/M_MED_PrawnTwine/Meshes/M_MED_PrawnTwine_AnimBP
target_skeleton=&/M_MED_PrawnTwine_Skeleton
graphs [
 AnimGraph {nodes [
   n1 = Control Rig!(Source=Copy Pose From Mesh() {bUseAttachedParent=true}, Alpha, Playing_Emote:bool=Get bIsPlayingEmote()) -> (Pose) {controlRigAssetReference {blueprintRigClass=&/M_MED_PrawnTwine_CtrlRig.M_MED_PrawnTwine_CtrlRig_C}}
   n2 = Anim Dynamics - Bone: dyn_tail!(ComponentPose=Local To Component(LocalPose=@n1), Alpha=1, GravityOverride:Vector="0,0,-980", ExternalForce="0,0,0") -> (Pose) {
    angularDampingOverride=0.9999
    angularSpringConstant=22000
    bAngularSpring=true
    bOverrideAngularDamping=true
    gravityScale=0.35
    simulationSpace=RootRelative
    boundBone {boneName=dyn_tail}
    componentAppliedLinearAccClamp=<2000 2000 2000>
    componentLinearAccScale=<1 1 1>
    componentLinearVelScale=<1 1 1>
    physicsBodyDefinitions [{
      boundBone {boneName=dyn_tail}
      boxExtents=<20 30 30>
      constraintSetup {angularLimitsMax=<0 20 20>, angularLimitsMin=<0 -20 -20>, angularTarget=<1 0 0>}
      localJointOffset=<3 0 0>
     }]
   }
   n3 = Anim Dynamics - Bone: dyn_tail_1!(ComponentPose=@n2, Alpha=1, GravityOverride:Vector="0,0,-980", ExternalForce="0,0,0") -> (Pose) {
    angularDampingOverride=0.9999
    angularSpringConstant=15000
    bAngularSpring=true
    bOverrideAngularDamping=true
    gravityScale=0.35
    simulationSpace=RootRelative
    boundBone {boneName=dyn_tail_1}
    componentAppliedLinearAccClamp=<2000 2000 2000>
    componentLinearAccScale=<1 1 1>
    componentLinearVelScale=<1 1 1>
    gravityOverride {z=-980}
    physicsBodyDefinitions [{
      boundBone {boneName=dyn_tail_1}
      boxExtents=<15 25 25>
      constraintSetup {angularLimitsMax=<0 20 20>, angularLimitsMin=<0 -20 -20>, angularTarget=<1 0 0>}
      localJointOffset=<4 0 0>
     }]
   }
   n4 = Anim Dynamics - Bone: dyn_tail_a_l!(ComponentPose=@n3, Alpha=1, GravityOverride:Vector="0,0,-980", ExternalForce="0,0,0") -> (Pose) {
    angularDampingOverride=0.9999
    angularSpringConstant=25000
    bAngularSpring=true
    bOverrideAngularDamping=true
    gravityScale=0.35
    simulationSpace=RootRelative
    boundBone {boneName=dyn_tail_a_l}
    componentAppliedLinearAccClamp=<2000 2000 2000>
    componentLinearAccScale=<1 1 1>
    componentLinearVelScale=<1 1 1>
    gravityOverride {z=-980}
    physicsBodyDefinitions [{
      boundBone {boneName=dyn_tail_a_l}
      boxExtents=<15 25 25>
      constraintSetup {angularLimitsMax=<0 10 10>, angularLimitsMin=<0 -10 -10>, angularTarget=<1 0 0>}
      localJointOffset=<2 0 0>
     }]
   }
   n5 = Anim Dynamics - Bone: dyn_tail_a_r!(ComponentPose=@n4, Alpha=1, GravityOverride:Vector="0,0,-980", ExternalForce="0,0,0") -> (Pose) {
    angularDampingOverride=0.9999
    angularSpringConstant=25000
    bAngularSpring=true
    bOverrideAngularDamping=true
    gravityScale=0.35
    simulationSpace=RootRelative
    boundBone {boneName=dyn_tail_a_r}
    componentAppliedLinearAccClamp=<2000 2000 2000>
    componentLinearAccScale=<1 1 1>
    componentLinearVelScale=<1 1 1>
    gravityOverride {z=-980}
    physicsBodyDefinitions [{
      boundBone {boneName=dyn_tail_a_r}
      boxExtents=<15 25 25>
      constraintSetup {angularLimitsMax=<0 10 10>, angularLimitsMin=<0 -10 -10>, angularTarget=<1 0 0>}
      localJointOffset=<2 0 0>
     }]
   }
   n6 = "Transform (Modify) Bone - Bone: base_tail"!(
    ComponentPose=@n5
    bAlphaBoolEnabled:bool=Get isWearingCape() -> (isWearingCape:bool)
    Translation="0,0,0"
    Rotation="0,0,0"
    Scale="0.75,0.75,0.75"
   ) -> (Pose) {
    alphaInputType=Bool
    rotationMode=BMM_Additive
    rotationSpace=BCS_BoneSpace
    scaleMode=BMM_Additive
    scaleSpace=BCS_BoneSpace
    translationMode=BMM_Additive
    translationSpace=BCS_BoneSpace
    boneToModify {boneName=base_tail}
    scale=<0.75 0.75 0.75>
   }
   Output Pose(Result=Component To Local(ComponentPose=@n6)) {layerGroup=None}
  ]}
 EventGraph {nodes [
   Event BlueprintUpdateAnimation()
   > Set RBANOff(RBANOff:bool=NOT Boolean(A=Get bRigidBodyEnabled())) -> (Output_Get:bool)
   Event OnCharacterCustomizationLoaded()
   n1 = CheckForBackbling()
   > Set isWearingPack(isWearingPack:bool=OR Boolean!(A=@n1, B=@n1.bIsBackblingACape, C:bool=@n1.bIsBackblingLong) -> (ReturnValue)) -> (Output_Get:bool)
   > Set isWearingCape(isWearingCape=OR Boolean!(A=@n1.bIsBackblingACape, B=@n1.bIsBackblingLong) -> (ReturnValue))
  ]}
]
variables [
 isWearingCape {type=boolean}
 isWearingPack {type=boolean}
 RBANOff {type=boolean}
]
