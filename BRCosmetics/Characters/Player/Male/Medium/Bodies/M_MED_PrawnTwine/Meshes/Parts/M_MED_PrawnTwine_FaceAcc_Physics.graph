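# Physics asset for the PrawnTwine male medium character's face accessory, giving its two ears jointed physics bodies anchored to a head capsule.
# Each side is a mirrored capsule chain from a small base through four dynamic ear segments.
# Bases rotate freely on the head, while segment joints allow no twist and only small swing that widens toward the tip, where the last joint's second swing limit drops to zero.
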PhysicsAsset /BRCosmetics/Characters/Player/Male/Medium/Bodies/M_MED_PrawnTwine/Meshes/Parts/M_MED_PrawnTwine_FaceAcc_Physics
bodies {
 base_ear_l [{shapeType=Capsule, length=2, radius=2, shapeName=base_ear_l_capsule, rotation=<rot 0 -90 0>}]
 base_ear_r [{shapeType=Capsule, length=2, radius=2, shapeName=base_ear_r_capsule, rotation=<rot 0 -90 0>}]
 dyn_ear_1_l [{shapeType=Capsule, length=8, radius=5, shapeName=dyn_ear_1_l_capsule, center=<2 0 0>, rotation=<rot 90 0 0>}]
 dyn_ear_1_r [{shapeType=Capsule, length=8, radius=5, shapeName=dyn_ear_1_r_capsule, center=<2 0 0>, rotation=<rot 90 0 0>}]
 dyn_ear_2_l [{shapeType=Capsule, length=8, radius=5, shapeName=dyn_ear_2_l_capsule, center=<2 0 0>, rotation=<rot 90 0 0>}]
 dyn_ear_2_r [{shapeType=Capsule, length=8, radius=5, shapeName=dyn_ear_2_r_capsule, center=<2 0 0>, rotation=<rot 90 0 0>}]
 dyn_ear_3_l [{shapeType=Capsule, length=8, radius=5, shapeName=dyn_ear_3_l_capsule, center=<2 0 0>, rotation=<rot 90 0 0>}]
 dyn_ear_3_r [{shapeType=Capsule, length=8, radius=5, shapeName=dyn_ear_3_r_capsule, center=<2 0 0>, rotation=<rot 90 0 0>}]
 dyn_ear_4_l [{shapeType=Capsule, length=8, radius=5, shapeName=dyn_ear_4_l_capsule, center=<2 0 0>, rotation=<rot 90 0 0>}]
 dyn_ear_4_r [{shapeType=Capsule, length=8, radius=5, shapeName=dyn_ear_4_r_capsule, center=<2 0 0>, rotation=<rot 90 0 0>}]
 head [{shapeType=Capsule, length=0.5237, radius=7.9314, shapeName=head_capsule, center=<15.2667 9.647 0.3473>, rotation=<rot 0 -118.8689 -89.4069>}]
}
constraints [
 {bone1Name=dyn_ear_2_l, bone2Name=dyn_ear_1_l, swing1LimitDegrees=4, swing2LimitDegrees=4, twistLimitDegrees=0}
 {bone1Name=dyn_ear_3_l, bone2Name=dyn_ear_2_l, swing1LimitDegrees=7, swing2LimitDegrees=7, twistLimitDegrees=0}
 {bone1Name=dyn_ear_4_l, bone2Name=dyn_ear_3_l, swing1LimitDegrees=10, swing2LimitDegrees=0, twistLimitDegrees=0}
 {bone1Name=dyn_ear_2_r, bone2Name=dyn_ear_1_r, swing1LimitDegrees=4, swing2LimitDegrees=4, twistLimitDegrees=0}
 {bone1Name=dyn_ear_3_r, bone2Name=dyn_ear_2_r, swing1LimitDegrees=7, swing2LimitDegrees=7, twistLimitDegrees=0}
 {bone1Name=dyn_ear_4_r, bone2Name=dyn_ear_3_r, swing1LimitDegrees=10, swing2LimitDegrees=0, twistLimitDegrees=0}
 {bone1Name=dyn_ear_1_l, bone2Name=base_ear_l, swing1LimitDegrees=4, swing2LimitDegrees=4, twistLimitDegrees=0}
 {bone1Name=dyn_ear_1_r, bone2Name=base_ear_r, swing1LimitDegrees=4, swing2LimitDegrees=4, twistLimitDegrees=0}
 {
  bone1Name=base_ear_l
  bone2Name=head
  swing1LimitDegrees=45
  swing1Motion=Free
  swing2LimitDegrees=45
  swing2Motion=Free
  twistLimitDegrees=45
  twistMotion=Free
 }
 {
  bone1Name=base_ear_r
  bone2Name=head
  swing1LimitDegrees=45
  swing1Motion=Free
  swing2LimitDegrees=45
  swing2Motion=Free
  twistLimitDegrees=45
  twistMotion=Free
 }
]
